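# Sound cue for the Season 15 quest-complete checkmark sound in the UI, routed through the Fort_UI sound class.
# A single checkmark wave plays through a modulator that applies a fixed slight pitch raise with no random variation.
# A concurrency override allows one instance at a time, stopping the oldest when a new one starts.
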SoundCue /Game/Sounds/Fort_UI_Sounds/S15/Quest_Complete_Cue
root=Modulator_0
nodes [
 {id=Modulator_0, children [WavePlayer_0], properties {pitchMax=1.02, pitchMin=1.02, volumeMax=1, volumeMin=1}}
 {id=WavePlayer_0, properties {soundWaveAssetPtr=&/S15_CheckMark}}
]
properties {
 bOverrideConcurrency=true
 maxDistance=2097152
 soundClassObject=&Fort_UI
 volumeMultiplier=1
 concurrencyOverrides {maxCount=1, resolutionRule=StopOldest}
}
